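; Buried chest dirt prop for the KeyDrop encounter that switches to a snow material when its nearest living world volume is tagged as snow.
; On BeginPlay, the authority scans all FortAthenaLivingWorldVolume actors for the closest by 2D distance.
; ValidateSnowTags checks that volume's filter tags against a search container and sets SpawnedInSnow on a match.
; The OnRep swaps the Dirt material on remote clients only.
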(event EventOnMaxDug
  (|CallParentFunction))

(event EventBeginPlay
  ;[1 1 1] On Start, Search for all LivingWorldVolumes and if the Cloeset is Snow, change the Material to Snow
  (|CallParentFunction)
  (Networking|SwitchHasAuthority
    (:Authority
      (Variables|Default|SetCurrentLocation (Transformation|GetActorLocation self))
      (bind _outactors (Actor|GetAllActorsOfClass "/Script/LagerRuntime.FortAthenaLivingWorldVolume"))
      (for _array_element _outactors
        (Utilities|IsValid (Variables|Default|GetClosestVolume)
          (:Is%20Valid
            (if (< (Math|Float|Absolute(Float) (Math|Vector|Distance2D(Vector) (Variables|Default|GetCurrentLocation) (Transformation|GetActorLocation _array_element))) (Math|Float|Absolute(Float) (Math|Vector|Distance2D(Vector) (Variables|Default|GetCurrentLocation) (Transformation|GetActorLocation (Variables|Default|GetClosestVolume)))))
              (Variables|Default|SetClosestVolume _array_element)))
          (:Is%20Not%20Valid
            (Variables|Default|SetClosestVolume _array_element))))
      (CallFunction|ValidateSnowTags))))

(event Custom|ValidateSnowTags
  (bind _closestvolume (Variables|Default|GetClosestVolume))
  ;[1 1 1] Validate that the Volume we found has the Tag we are looking for.
  (Utilities|IsValid _closestvolume
    (:Is%20Valid
      (if (GameplayTags|HasAnyTags (Class|FortAthenaLivingWorldPointProviderInterface|GetFiltersTags _closestvolume) (Variables|Default|GetTagContainerToSearchFor))
        (Variables|Default|SetSpawnedInSnow true)))))

(fn OnRep_SpawnedInSnow
  (Networking|SwitchHasAuthority
    (:Remote
      (Rendering|Material|SetMaterial (Variables|Default|GetDirt) 0 "/Game/Effects/Fort_Effects/Gameplay/PirateChests/M_Chest_Snow.M_Chest_Snow"))))

(fn ConstructionScript
  (|CallParentFunction))
